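# Physics asset for the Teriyaki Fish Toon Male pickaxe, a two-handed melee harvesting tool cosmetic.
# It has two capsule bodies: a long one for the weapon handle and a shorter, wider, offset one for the dyn_neckHinge bone.
# A single constraint joins the hinge to the handle, almost locking swing and allowing limited twist, so the jointed part moves like a restricted hinge.
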PhysicsAsset /BRCosmetics/Weapons/FORT_Melee/Pickaxe_Teriyaki_Fish_Toon_Male/Meshes/Teriyaki_Fish_Toon_Male_Axe_Physics
bodies {
 Melee_TwoHanded_Handle [{shapeType=Capsule, length=100, radius=10}]
 dyn_neckHinge [{shapeType=Capsule, length=32.6939, radius=22.6418, center=<-1.7099 3.9574 31.0552>}]
}
constraints [{bone1Name=dyn_neckHinge, bone2Name=Melee_TwoHanded_Handle, swing1LimitDegrees=0.5, swing2LimitDegrees=0.5, twistLimitDegrees=25}]
