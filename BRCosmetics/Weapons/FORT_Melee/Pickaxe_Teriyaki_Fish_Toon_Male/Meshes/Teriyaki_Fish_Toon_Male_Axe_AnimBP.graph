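# Post-process-style AnimBP for the Teriyaki Fish Toon pickaxe, animating its dangling strings and phase-dependent look.
# Additive bone offsets (active below LOD 2) bend four string chains, then per-bone AnimDynamics springs add sway, followed by RigidBody simulation enabled only in phase 3.
# Each update reads PhaseSelection from the owning pickaxe actor to set material "Hide Element" scalars, with a delayed random blink swapping elements in phases 1 and 2.
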AnimBlueprint /BRCosmetics/Weapons/FORT_Melee/Pickaxe_Teriyaki_Fish_Toon_Male/Meshes/Teriyaki_Fish_Toon_Male_Axe_AnimBP
target_skeleton=&/Teriyaki_Fish_Toon_Male_Axe_Skeleton
graphs [
 AnimGraph {nodes [
   AnimGraphNode_SubInput_0 = Input Pose()
   n1 = Get stringRotationDisabled() -> (stringRotationDisabled:double)
   n2 = "Transform (Modify) Bone - Bone: dyn_stringA_1"(ComponentPose=Local To Component(LocalPose=@AnimGraphNode_SubInput_0), Alpha=@n1, Translation="0,0,0", Rotation="-1,-2,0", Scale="1,1,1") {
    rotationMode=BMM_Additive
    rotationSpace=BCS_BoneSpace
    alphaBoolBlend {blendOption=Linear}
    boneToModify {boneName=dyn_stringA_1}
    rotation {pitch=-1, yaw=-2}
   }
   n3 = "Transform (Modify) Bone - Bone: dyn_stringA_2"(ComponentPose=@n2, Alpha=@n1, Translation="0,0,0", Rotation="-3,0,0", Scale="1,1,1") {
    rotationMode=BMM_Additive
    rotationSpace=BCS_BoneSpace
    alphaBoolBlend {blendOption=Linear}
    boneToModify {boneName=dyn_stringA_2}
    rotation {pitch=-3}
   }
   n4 = "Transform (Modify) Bone - Bone: dyn_stringA_3"(ComponentPose=@n3, Alpha=@n1, Translation="0,0,0", Rotation="-3,0,0", Scale="1,1,1") {
    rotationMode=BMM_Additive
    rotationSpace=BCS_BoneSpace
    alphaBoolBlend {blendOption=Linear}
    boneToModify {boneName=dyn_stringA_3}
    rotation {pitch=-3}
   }
   n5 = "Transform (Modify) Bone - Bone: dyn_stringA_4"(ComponentPose=@n4, Alpha=@n1, Translation="0,0,0", Rotation="-3,-3,0", Scale="1,1,1") {
    rotationMode=BMM_Additive
    rotationSpace=BCS_BoneSpace
    alphaBoolBlend {blendOption=Linear}
    boneToModify {boneName=dyn_stringA_4}
    rotation {pitch=-3, yaw=-3}
   }
   n6 = "Transform (Modify) Bone - Bone: dyn_stringA_5"(ComponentPose=@n5, Alpha=@n1, Translation="0,0,0", Rotation="-3,3,0", Scale="1,1,1") {
    rotationMode=BMM_Additive
    rotationSpace=BCS_BoneSpace
    alphaBoolBlend {blendOption=Linear}
    boneToModify {boneName=dyn_stringA_5}
    rotation {pitch=-3, yaw=3}
   }
   n7 = "Transform (Modify) Bone - Bone: dyn_stringA_6"(ComponentPose=@n6, Alpha=@n1, Translation="0,0,0", Rotation="-3,6,0", Scale="1,1,1") {
    rotationMode=BMM_Additive
    rotationSpace=BCS_BoneSpace
    alphaBoolBlend {blendOption=Linear}
    boneToModify {boneName=dyn_stringA_6}
    rotation {pitch=-3, yaw=6}
   }
   n8 = "Transform (Modify) Bone - Bone: dyn_stringB_1"(ComponentPose=@n7, Alpha=@n1, Translation="0,0,0", Rotation="-3,0,0", Scale="1,1,1") {
    rotationMode=BMM_Additive
    rotationSpace=BCS_BoneSpace
    alphaBoolBlend {blendOption=Linear}
    boneToModify {boneName=dyn_stringB_1}
    rotation {pitch=-3}
   }
   n9 = "Transform (Modify) Bone - Bone: dyn_stringB_2"(ComponentPose=@n8, Alpha=@n1, Translation="0,0,0", Rotation="-1,2,0", Scale="1,1,1") {
    rotationMode=BMM_Additive
    rotationSpace=BCS_BoneSpace
    alphaBoolBlend {blendOption=Linear}
    boneToModify {boneName=dyn_stringB_2}
    rotation {pitch=-1, yaw=2}
   }
   n10 = "Transform (Modify) Bone - Bone: dyn_stringB_3"(ComponentPose=@n9, Alpha=@n1, Translation="0,0,0", Rotation="-4,5,0", Scale="1,1,1") {
    rotationMode=BMM_Additive
    rotationSpace=BCS_BoneSpace
    alphaBoolBlend {blendOption=Linear}
    boneToModify {boneName=dyn_stringB_3}
    rotation {pitch=-4, yaw=5}
   }
   n11 = "Transform (Modify) Bone - Bone: dyn_stringB_4"(ComponentPose=@n10, Alpha=@n1, Translation="0,0,0", Rotation="-6,-5,0", Scale="1,1,1") {
    rotationMode=BMM_Additive
    rotationSpace=BCS_BoneSpace
    alphaBoolBlend {blendOption=Linear}
    boneToModify {boneName=dyn_stringB_4}
    rotation {pitch=-6, yaw=-5}
   }
   n12 = "Transform (Modify) Bone - Bone: dyn_stringB_5"(ComponentPose=@n11, Alpha=@n1, Translation="0,0,0", Rotation="-10,-5,0", Scale="1,1,1") {
    rotationMode=BMM_Additive
    rotationSpace=BCS_BoneSpace
    alphaBoolBlend {blendOption=Linear}
    boneToModify {boneName=dyn_stringB_5}
    rotation {pitch=-10, yaw=-5}
   }
   n13 = "Transform (Modify) Bone - Bone: dyn_stringB_6"(ComponentPose=@n12, Alpha=@n1, Translation="0,0,0", Rotation="-8,0,0", Scale="1,1,1") {
    rotationMode=BMM_Additive
    rotationSpace=BCS_BoneSpace
    alphaBoolBlend {blendOption=Linear}
    boneToModify {boneName=dyn_stringB_6}
    rotation {pitch=-8}
   }
   n14 = "Transform (Modify) Bone - Bone: dyn_stringB_7"(ComponentPose=@n13, Alpha=@n1, Translation="0,0,0", Rotation="-30,0,0", Scale="1,1,1") {
    rotationMode=BMM_Additive
    rotationSpace=BCS_BoneSpace
    alphaBoolBlend {blendOption=Linear}
    boneToModify {boneName=dyn_stringB_7}
    rotation {pitch=-30}
   }
   n15 = "Transform (Modify) Bone - Bone: dyn_stringC_1"(ComponentPose=@n14, Alpha=@n1, Translation="0,0,0", Rotation="-30,0,0", Scale="1,1,1") {
    rotationMode=BMM_Additive
    rotationSpace=BCS_BoneSpace
    alphaBoolBlend {blendOption=Linear}
    boneToModify {boneName=dyn_stringC_1}
    rotation {pitch=-30}
   }
   n16 = "Transform (Modify) Bone - Bone: dyn_stringC_2"(ComponentPose=@n15, Alpha=@n1, Translation="0,0,0", Rotation="-6,-5,0", Scale="1,1,1") {
    rotationMode=BMM_Additive
    rotationSpace=BCS_BoneSpace
    alphaBoolBlend {blendOption=Linear}
    boneToModify {boneName=dyn_stringC_2}
    rotation {pitch=-6, yaw=-5}
   }
   n17 = "Transform (Modify) Bone - Bone: dyn_stringC_3"(ComponentPose=@n16, Alpha=@n1, Translation="0,0,0", Rotation="-5,0,0", Scale="1,1,1") {
    rotationMode=BMM_Additive
    rotationSpace=BCS_BoneSpace
    alphaBoolBlend {blendOption=Linear}
    boneToModify {boneName=dyn_stringC_3}
    rotation {pitch=-5}
   }
   n18 = "Transform (Modify) Bone - Bone: dyn_stringD_1"(ComponentPose=@n17, Alpha=@n1, Translation="0,0,0", Rotation="-28,5,0", Scale="1,1,1") {
    rotationMode=BMM_Additive
    rotationSpace=BCS_BoneSpace
    alphaBoolBlend {blendOption=Linear}
    boneToModify {boneName=dyn_stringD_1}
    rotation {pitch=-28, yaw=5}
   }
   n19 = "Transform (Modify) Bone - Bone: dyn_stringD_2"(ComponentPose=@n18, Alpha=@n1, Translation="0,0,0", Rotation="-10,7,0", Scale="1,1,1") {
    rotationMode=BMM_Additive
    rotationSpace=BCS_BoneSpace
    alphaBoolBlend {blendOption=Linear}
    boneToModify {boneName=dyn_stringD_2}
    rotation {pitch=-10, yaw=7}
   }
   n20 = "Transform (Modify) Bone - Bone: dyn_stringD_3"(ComponentPose=@n19, Alpha=@n1, Translation="0,0,0", Rotation="-35,9,0", Scale="1,1,1") {
    rotationMode=BMM_Additive
    rotationSpace=BCS_BoneSpace
    alphaBoolBlend {blendOption=Linear}
    boneToModify {boneName=dyn_stringD_3}
    rotation {pitch=-35, yaw=9}
   }
   n21 = Anim Dynamics - Bone: dyn_stringA_1(ComponentPose=@n20, Alpha=1, ExternalForce="0,0,0") {
    angularSpringConstant=100
    bAngularSpring=true
    bGravityOverrideInSimSpace=true
    simulationSpace=BoneRelative
    alphaBoolBlend {blendOption=Linear}
    alphaScaleBiasClamp {scale=0.4}
    boundBone {boneName=dyn_stringA_1}
    componentLinearAccScale=<1.3 1.3 1.3>
    physicsBodyDefinitions [{
      sphereCollisionRadius=0
      boundBone {boneName=dyn_stringA_1}
      boxExtents=<2 2 2>
      constraintSetup {angularLimitsMax=<0 10 5>, angularLimitsMin=<0 0 0>, angularTarget=<0 0 0>}
      localJointOffset=<0 0 0.003>
     }]
   }
   n22 = Anim Dynamics - Bone: dyn_stringA_2(ComponentPose=@n21, Alpha=1, ExternalForce="0,0,0") {
    angularSpringConstant=100
    bAngularSpring=true
    bGravityOverrideInSimSpace=true
    simulationSpace=BoneRelative
    alphaBoolBlend {blendOption=Linear}
    alphaScaleBiasClamp {scale=0.4}
    boundBone {boneName=dyn_stringA_2}
    componentLinearAccScale=<1.3 1.3 1.3>
    physicsBodyDefinitions [{
      sphereCollisionRadius=0
      boundBone {boneName=dyn_stringA_2}
      boxExtents=<2 2 2>
      constraintSetup {angularLimitsMax=<0 10 5>, angularLimitsMin=<0 0 0>, angularTarget=<0 0 0>}
      localJointOffset=<0 0 0.003>
     }]
   }
   n23 = Anim Dynamics - Bone: dyn_stringA_3(ComponentPose=@n22, Alpha=1, ExternalForce="0,0,0") {
    angularSpringConstant=100
    bAngularSpring=true
    bGravityOverrideInSimSpace=true
    simulationSpace=BoneRelative
    alphaBoolBlend {blendOption=Linear}
    alphaScaleBiasClamp {scale=0.4}
    boundBone {boneName=dyn_stringA_3}
    componentLinearAccScale=<1.3 1.3 1.3>
    physicsBodyDefinitions [{
      sphereCollisionRadius=0
      boundBone {boneName=dyn_stringA_3}
      boxExtents=<2 2 2>
      constraintSetup {angularLimitsMax=<0 10 5>, angularLimitsMin=<0 0 0>, angularTarget=<0 0 0>}
      localJointOffset=<0 0 0.003>
     }]
   }
   n24 = Anim Dynamics - Bone: dyn_stringA_4(ComponentPose=@n23, Alpha=1, ExternalForce="0,0,0") {
    angularSpringConstant=100
    bAngularSpring=true
    bGravityOverrideInSimSpace=true
    simulationSpace=BoneRelative
    alphaBoolBlend {blendOption=Linear}
    alphaScaleBiasClamp {scale=0.4}
    boundBone {boneName=dyn_stringA_4}
    componentLinearAccScale=<1.3 1.3 1.3>
    physicsBodyDefinitions [{
      sphereCollisionRadius=0
      boundBone {boneName=dyn_stringA_4}
      boxExtents=<2 2 2>
      constraintSetup {angularLimitsMax=<0 10 5>, angularLimitsMin=<0 0 0>, angularTarget=<0 0 0>}
      localJointOffset=<0 0 0.003>
     }]
   }
   n25 = Anim Dynamics - Bone: dyn_stringA_5(ComponentPose=@n24, Alpha=1, ExternalForce="0,0,0") {
    angularSpringConstant=100
    bAngularSpring=true
    bGravityOverrideInSimSpace=true
    simulationSpace=BoneRelative
    alphaBoolBlend {blendOption=Linear}
    alphaScaleBiasClamp {scale=0.4}
    boundBone {boneName=dyn_stringA_5}
    componentLinearAccScale=<1.3 1.3 1.3>
    physicsBodyDefinitions [{
      sphereCollisionRadius=0
      boundBone {boneName=dyn_stringA_5}
      boxExtents=<2 2 2>
      constraintSetup {angularLimitsMax=<0 10 5>, angularLimitsMin=<0 0 0>, angularTarget=<0 0 0>}
      localJointOffset=<0 0 0.003>
     }]
   }
   n26 = Anim Dynamics - Bone: dyn_stringA_6(ComponentPose=@n25, Alpha=1, ExternalForce="0,0,0") {
    angularSpringConstant=100
    bAngularSpring=true
    bGravityOverrideInSimSpace=true
    simulationSpace=BoneRelative
    alphaBoolBlend {blendOption=Linear}
    alphaScaleBiasClamp {scale=0.4}
    boundBone {boneName=dyn_stringA_6}
    componentLinearAccScale=<1.3 1.3 1.3>
    physicsBodyDefinitions [{
      sphereCollisionRadius=0
      boundBone {boneName=dyn_stringA_6}
      boxExtents=<2 2 2>
      constraintSetup {angularLimitsMax=<0 10 5>, angularLimitsMin=<0 0 0>, angularTarget=<0 0 0>}
      localJointOffset=<0 0 0.003>
     }]
   }
   n27 = Anim Dynamics - Bone: dyn_stringB_1(ComponentPose=@n26, Alpha=1, ExternalForce="0,0,0") {
    angularSpringConstant=60
    bAngularSpring=true
    bGravityOverrideInSimSpace=true
    simulationSpace=BoneRelative
    alphaBoolBlend {blendOption=Linear}
    boundBone {boneName=dyn_stringB_1}
    componentLinearAccScale=<1.1 1.1 1.1>
    physicsBodyDefinitions [{
      sphereCollisionRadius=0
      boundBone {boneName=dyn_stringB_1}
      boxExtents=<2 2 2>
      constraintSetup {angularLimitsMax=<0 12 5>, angularLimitsMin=<0 0 0>, angularTarget=<0 0 0>}
      localJointOffset=<0 0 0.003>
     }]
   }
   n28 = Anim Dynamics - Bone: dyn_stringB_2(ComponentPose=@n27, Alpha=1, ExternalForce="0,0,0") {
    angularSpringConstant=60
    bAngularSpring=true
    bGravityOverrideInSimSpace=true
    simulationSpace=BoneRelative
    alphaBoolBlend {blendOption=Linear}
    boundBone {boneName=dyn_stringB_2}
    componentLinearAccScale=<1.1 1.1 1.1>
    physicsBodyDefinitions [{
      sphereCollisionRadius=0
      boundBone {boneName=dyn_stringB_2}
      boxExtents=<2 2 2>
      constraintSetup {angularLimitsMax=<0 12 5>, angularLimitsMin=<0 0 0>, angularTarget=<0 0 0>}
      localJointOffset=<0 0 0.003>
     }]
   }
   n29 = Anim Dynamics - Bone: dyn_stringB_3(ComponentPose=@n28, Alpha=1, ExternalForce="0,0,0") {
    angularSpringConstant=60
    bAngularSpring=true
    bGravityOverrideInSimSpace=true
    simulationSpace=BoneRelative
    alphaBoolBlend {blendOption=Linear}
    boundBone {boneName=dyn_stringB_3}
    componentLinearAccScale=<1.1 1.1 1.1>
    physicsBodyDefinitions [{
      sphereCollisionRadius=0
      boundBone {boneName=dyn_stringB_3}
      boxExtents=<2 2 2>
      constraintSetup {angularLimitsMax=<0 12 5>, angularLimitsMin=<0 0 0>, angularTarget=<0 0 0>}
      localJointOffset=<0 0 0.003>
     }]
   }
   n30 = Anim Dynamics - Bone: dyn_stringB_4(ComponentPose=@n29, Alpha=1, ExternalForce="0,0,0") {
    angularSpringConstant=60
    bAngularSpring=true
    bGravityOverrideInSimSpace=true
    simulationSpace=BoneRelative
    alphaBoolBlend {blendOption=Linear}
    boundBone {boneName=dyn_stringB_4}
    componentLinearAccScale=<1.1 1.1 1.1>
    physicsBodyDefinitions [{
      sphereCollisionRadius=0
      boundBone {boneName=dyn_stringB_4}
      boxExtents=<2 2 2>
      constraintSetup {angularLimitsMax=<0 12 5>, angularLimitsMin=<0 0 0>, angularTarget=<0 0 0>}
      localJointOffset=<0 0 0.003>
     }]
   }
   n31 = Anim Dynamics - Bone: dyn_stringB_5(ComponentPose=@n30, Alpha=1, ExternalForce="0,0,0") {
    angularSpringConstant=60
    bAngularSpring=true
    bGravityOverrideInSimSpace=true
    simulationSpace=BoneRelative
    alphaBoolBlend {blendOption=Linear}
    boundBone {boneName=dyn_stringB_5}
    componentLinearAccScale=<1.1 1.1 1.1>
    physicsBodyDefinitions [{
      sphereCollisionRadius=0
      boundBone {boneName=dyn_stringB_5}
      boxExtents=<2 2 2>
      constraintSetup {angularLimitsMax=<0 12 5>, angularLimitsMin=<0 0 0>, angularTarget=<0 0 0>}
      localJointOffset=<0 0 0.003>
     }]
   }
   n32 = Anim Dynamics - Bone: dyn_stringB_6(ComponentPose=@n31, Alpha=1, ExternalForce="0,0,0") {
    angularSpringConstant=60
    bAngularSpring=true
    bGravityOverrideInSimSpace=true
    simulationSpace=BoneRelative
    alphaBoolBlend {blendOption=Linear}
    boundBone {boneName=dyn_stringB_6}
    componentLinearAccScale=<1.1 1.1 1.1>
    physicsBodyDefinitions [{
      sphereCollisionRadius=0
      boundBone {boneName=dyn_stringB_6}
      boxExtents=<2 2 2>
      constraintSetup {angularLimitsMax=<0 12 5>, angularLimitsMin=<0 0 0>, angularTarget=<0 0 0>}
      localJointOffset=<0 0 0.003>
     }]
   }
   n33 = Anim Dynamics - Bone: dyn_stringB_7(ComponentPose=@n32, Alpha=1, ExternalForce="0,0,0") {
    angularSpringConstant=60
    bAngularSpring=true
    bGravityOverrideInSimSpace=true
    simulationSpace=BoneRelative
    alphaBoolBlend {blendOption=Linear}
    boundBone {boneName=dyn_stringB_7}
    componentLinearAccScale=<1.1 1.1 1.1>
    physicsBodyDefinitions [{
      sphereCollisionRadius=0
      boundBone {boneName=dyn_stringB_7}
      boxExtents=<2 2 2>
      constraintSetup {angularLimitsMax=<0 12 5>, angularLimitsMin=<0 0 0>, angularTarget=<0 0 0>}
      localJointOffset=<0 0 0.003>
     }]
   }
   n34 = Anim Dynamics - Bone: dyn_stringC_1(ComponentPose=@n33, Alpha=1, ExternalForce="0,0,0") {
    angularSpringConstant=100
    bAngularSpring=true
    bGravityOverrideInSimSpace=true
    simulationSpace=BoneRelative
    alphaBoolBlend {blendOption=Linear}
    boundBone {boneName=dyn_stringC_1}
    componentLinearAccScale=<1.5 1.5 1.5>
    physicsBodyDefinitions [{
      sphereCollisionRadius=0
      boundBone {boneName=dyn_stringC_1}
      boxExtents=<2 2 2>
      constraintSetup {angularLimitsMax=<0 12 5>, angularLimitsMin=<0 0 0>, angularTarget=<0 0 0>}
      localJointOffset=<0 0 0.003>
     }]
   }
   n35 = Anim Dynamics - Bone: dyn_stringC_2(ComponentPose=@n34, Alpha=1, ExternalForce="0,0,0") {
    angularSpringConstant=100
    bAngularSpring=true
    bGravityOverrideInSimSpace=true
    simulationSpace=BoneRelative
    alphaBoolBlend {blendOption=Linear}
    boundBone {boneName=dyn_stringC_2}
    componentLinearAccScale=<1.5 1.5 1.5>
    physicsBodyDefinitions [{
      sphereCollisionRadius=0
      boundBone {boneName=dyn_stringC_2}
      boxExtents=<2 2 2>
      constraintSetup {angularLimitsMax=<0 12 5>, angularLimitsMin=<0 0 0>, angularTarget=<0 0 0>}
      localJointOffset=<0 0 0.003>
     }]
   }
   n36 = Anim Dynamics - Bone: dyn_stringC_3(ComponentPose=@n35, Alpha=1, ExternalForce="0,0,0") {
    angularSpringConstant=100
    bAngularSpring=true
    bGravityOverrideInSimSpace=true
    simulationSpace=BoneRelative
    alphaBoolBlend {blendOption=Linear}
    boundBone {boneName=dyn_stringC_3}
    componentLinearAccScale=<1.5 1.5 1.5>
    physicsBodyDefinitions [{
      sphereCollisionRadius=0
      boundBone {boneName=dyn_stringC_3}
      boxExtents=<2 2 2>
      constraintSetup {angularLimitsMax=<0 12 5>, angularLimitsMin=<0 0 0>, angularTarget=<0 0 0>}
      localJointOffset=<0 0 0.003>
     }]
   }
   n37 = Anim Dynamics - Bone: dyn_stringD_1(ComponentPose=@n36, Alpha=1, ExternalForce="0,0,0") {
    angularSpringConstant=60
    bAngularSpring=true
    bGravityOverrideInSimSpace=true
    simulationSpace=BoneRelative
    alphaBoolBlend {blendOption=Linear}
    boundBone {boneName=dyn_stringD_1}
    componentLinearAccScale=<1.2 1.2 1.2>
    physicsBodyDefinitions [{
      sphereCollisionRadius=0
      boundBone {boneName=dyn_stringD_1}
      boxExtents=<2 2 2>
      constraintSetup {angularLimitsMax=<0 10 5>, angularLimitsMin=<0 0 0>, angularTarget=<0 0 0>}
      localJointOffset=<0 0 0.003>
     }]
   }
   n38 = Anim Dynamics - Bone: dyn_stringD_2(ComponentPose=@n37, Alpha=1, ExternalForce="0,0,0") {
    angularSpringConstant=60
    bAngularSpring=true
    bGravityOverrideInSimSpace=true
    simulationSpace=BoneRelative
    alphaBoolBlend {blendOption=Linear}
    boundBone {boneName=dyn_stringD_2}
    componentLinearAccScale=<1.2 1.2 1.2>
    physicsBodyDefinitions [{
      sphereCollisionRadius=0
      boundBone {boneName=dyn_stringD_2}
      boxExtents=<2 2 2>
      constraintSetup {angularLimitsMax=<0 10 5>, angularLimitsMin=<0 0 0>, angularTarget=<0 0 0>}
      localJointOffset=<0 0 0.003>
     }]
   }
   n39 = Anim Dynamics - Bone: dyn_stringD_3(ComponentPose=@n38, Alpha=1, ExternalForce="0,0,0") {
    angularSpringConstant=60
    bAngularSpring=true
    bGravityOverrideInSimSpace=true
    simulationSpace=BoneRelative
    alphaBoolBlend {blendOption=Linear}
    boundBone {boneName=dyn_stringD_3}
    componentLinearAccScale=<1.2 1.2 1.2>
    physicsBodyDefinitions [{
      sphereCollisionRadius=0
      boundBone {boneName=dyn_stringD_3}
      boxExtents=<2 2 2>
      constraintSetup {angularLimitsMax=<0 10 5>, angularLimitsMin=<0 0 0>, angularTarget=<0 0 0>}
      localJointOffset=<0 0 0.003>
     }]
   }
   n40 = RigidBody(ComponentPose=@n39, Alpha=Get phase3dynamics() -> (phase3dynamics:double)) {
    bClampLinearTranslationLimitToRefPose=true
    overridePhysicsAsset=&/Teriyaki_Fish_Toon_Male_Axe_Physics
    alphaBoolBlend {blendOption=Linear}
    alphaScaleBiasClamp {bInterpResult=true}
    baseBoneRef {boneName=Melee_TwoHanded_Handle}
    componentLinearAccScale=<0.2 0.2 0.2>
    componentLinearVelScale=<0.3 0.3 0.3>
   }
   Output Pose(Result=Component To Local(ComponentPose=@n40)) {layerGroup=None}
  ]}
 EventGraph {nodes [
   Event BlueprintUpdateAnimation()
   > n1 = Cast To B_Melee_Impact_Pickaxe_Teriyaki_Fish_Toon!(execute, Object=GetOwningActor()) -> (then, CastFailed, AsB Melee Impact Pickaxe Teriyaki Fish Toon:B Melee Impact Pickaxe Teriyaki Fish Toon Object Reference) {is=title K2Node_DynamicCast}
   > n2 = Sequence!(execute) -> (then_0, then_1, then_2:Exec, then_3:Exec)
   n3 = Get PhaseSelection!(self:B Melee Impact Pickaxe Teriyaki Fish Toon Object Reference=@n1) -> (PhaseSelection:int=0)
   > n4 = Switch on Int!(execute:Exec, Selection:int=@n3) -> (Default:Exec, 0:Exec, 1:Exec, 2:Exec, 3:Exec)
   n5 = GetOwningComponent()
   n6 = Get Dynamic Material Instance(execute=@n4.1, SkeletalMeshComponent=@n5, ElementIndex=1)
   > n7 = SetScalarParameterValue(self=@n6, ParameterName=Hide Element 01, Value=0)
   n8 = Set blinkSwitch(execute=@n14.then, blinkSwitch:double="To Float (Boolean)"(InBool=RandomBoolWithWeight!(Weight:float=0.05) -> (ReturnValue:bool))) -> (Output_Get:double)
   SetScalarParameterValue(execute=@n7.then, self=@n6, ParameterName=Hide Element 02, Value=@n8)
   n9 = float + float(A=float * float(A=@n8, B=-1), B=1) {is=title K2Node_CommutativeAssociativeBinaryOperator}
   > SetScalarParameterValue(self=@n6, ParameterName=Hide Element 03, Value=@n9)
   > SetScalarParameterValue(self=@n6, ParameterName=Hide Element 04, Value=1)
   > SetScalarParameterValue(self=@n6, ParameterName=Hide Element 05, Value=1)
   > SetScalarParameterValue(self=@n6, ParameterName=Hide Element 06, Value=1)
   > SetScalarParameterValue(self=@n6, ParameterName=Hide Element 07, Value=1)
   n10 = Get Dynamic Material Instance(execute=@n4.2, SkeletalMeshComponent=@n5, ElementIndex=1)
   > SetScalarParameterValue(self=@n10, ParameterName=Hide Element 01, Value=1)
   > SetScalarParameterValue(self=@n10, ParameterName=Hide Element 02, Value=1)
   > SetScalarParameterValue(self=@n10, ParameterName=Hide Element 03, Value=1)
   > SetScalarParameterValue(self=@n10, ParameterName=Hide Element 04, Value=0)
   > SetScalarParameterValue(self=@n10, ParameterName=Hide Element 05, Value=@n8)
   > SetScalarParameterValue(self=@n10, ParameterName=Hide Element 06, Value=@n9)
   > SetScalarParameterValue(self=@n10, ParameterName=Hide Element 07, Value=1)
   n11 = Get Dynamic Material Instance(execute=@n4.3, SkeletalMeshComponent=@n5, ElementIndex=1)
   > SetScalarParameterValue(self=@n11, ParameterName=Hide Element 01, Value=1)
   > SetScalarParameterValue(self=@n11, ParameterName=Hide Element 02, Value=1)
   > SetScalarParameterValue(self=@n11, ParameterName=Hide Element 03, Value=1)
   > SetScalarParameterValue(self=@n11, ParameterName=Hide Element 04, Value=1)
   > SetScalarParameterValue(self=@n11, ParameterName=Hide Element 05, Value=1)
   > SetScalarParameterValue(self=@n11, ParameterName=Hide Element 06, Value=1)
   > SetScalarParameterValue(self=@n11, ParameterName=Hide Element 07, Value=0)
   n12 = Branch(execute=@n2.then_1, Condition="Equal (Integer)"(A=@n3, B=3) {is=title K2Node_CallFunction})
   > Set phase3dynamics(phase3dynamics:double=1) -> (Output_Get:double)
   Set phase3dynamics(execute=@n12.else, phase3dynamics=0)
   n13 = Get CurrentLOD()
   Set stringRotationDisabled(
    execute=@n2.then_3
    stringRotationDisabled:double="To Float (Boolean)"(InBool="integer < integer"!(A:int=@n13, B:int=2) -> (ReturnValue:bool) {is=title K2Node_CallFunction})
   ) -> (Output_Get:double)
   n14 = Delay(execute=@n2.then_2, Duration=0.15)
  ]}
]
variables [
 blinkSwitch {type=number}
 completed {type=boolean}
 Phase Selection {type=integer}
 phase3dynamics {type=number}
 stringRotationDisabled {type=number}
 WeaponMesh {type=SkeletalMeshComponent}
]
